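# PistolCanary comic border widget that fades a comic-panel border image in and out.
# Its only widget is a root Image that draws the border texture as a nine-slice box and is set fully transparent on Construct.
# ComicBorderFadeIn and ComicBorderFadeOut events play matching half-second animations that ramp the image's render opacity up or down.
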WidgetBlueprint /PistolCanary/Layout/BP/ComicPanels/HeroPanels/PistolCanary_ComicBorder
animations [
 FadeIn {
  binding_count=1
  length_seconds=0.5
  object=&self:FadeIn
  sequence {
   bindings [Image_160 {
     id=#1
     possessed_class=Image
     tracks [{
       type=MovieSceneFloatTrack
       display_name=Render Opacity
       sections [{type=MovieSceneFloatSection, end_frame=30, start_frame=0, channels [None {type=MovieSceneScriptingFloatChannel, keys [[0,0],[30,1]]}]}]
      }]
    }]
   display_rate {numerator=60}
  }
 }
 FadeOut {
  binding_count=1
  length_seconds=0.5
  object=&self:FadeOut
  sequence {
   bindings [Image_160 {
     id=#2
     possessed_class=Image
     tracks [{
       type=MovieSceneFloatTrack
       display_name=Render Opacity
       sections [{type=MovieSceneFloatSection, end_frame=30, start_frame=0, channels [None {type=MovieSceneScriptingFloatChannel, keys [[0,1],[30,0]]}]}]
      }]
    }]
   display_rate {numerator=60}
  }
 }
]
graphs [EventGraph {nodes [
   Event Construct()
   > SetRenderOpacity(self=Get Image_160() -> (Image_160:Image Object Reference), InOpacity=0)
   Event ComicBorderFadeIn() {is=title K2Node_Event}
   > PlayAnimation(InAnimation=Get FadeIn() -> (FadeIn:Widget Animation Object Reference))
   Event ComicBorderFadeOut() {is=title K2Node_Event}
   > PlayAnimation(InAnimation=Get FadeOut() -> (FadeOut:Widget Animation Object Reference))
   Event PreConstruct()
   Event Tick()
  ]}]
variables []
widgets {
 info {inheritedWidgetCount=0, parentClass=&UserWidget, rootWidgetClass=&Image, widgetCount=1}
 widgets [Image_160 {
   widgetClassPath=&Image
   bIsVariable=true
   widget=Image_160
   properties {
    visibility=SelfHitTestInvisible
    brush {
     bIsDynamicallyLoaded=false
     drawAs=Box
     imageType=FullColor
     mirroring=NoMirror
     resourceName=None
     resourceObject=/PistolCanary/Layout/ComicPanels_v2/T_ComicBorder_256__1_
     tiling=NoTile
     imageSize=<256 256>
     margin=<ltrb 0.2 0.2 0.2 0.2>
     outlineSettings {
      bUseBrushTransparency=false
      roundingType=HalfHeightRadius
      width=0
      color {colorUseRule=UseColor_Specified, specifiedColor=<rgba 0 0 0 0>}
      cornerRadii=<0 0 0 0>
     }
     tintColor {colorUseRule=UseColor_Specified, specifiedColor=<rgba 1 1 1 1>}
     uVRegion {bIsValid=false, max=<0 0>, min=<0 0>}
    }
   }
  }]
}
