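# PistolCanary comic overlay widget showing Hero 1 left and right panels, each with background and character images.
# Construct hides the images; the play event picks LeftPanelAnim or RightPanelAnim by the HeroPanel enum, ignoring PanelSide.
# Animations slide and fade each canvas in and out while the character slowly scales or drifts.
# Load fills brushes from soft textures, low-res on mobile; unload clears them, forces garbage collection and prints a message.
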WidgetBlueprint /PistolCanary/Layout/BP/ComicPanels/HeroPanels/PistolCanary_HeroPanel1
animations [
 LeftPanelAnim {
  binding_count=3
  length_seconds=7.6333
  object=&self:LeftPanelAnim
  sequence {
   bindings [
    LeftPanel_BG {
     id=#1
     possessed_class=Image
     tracks [{
       type=MovieSceneFloatTrack
       display_name=Render Opacity
       sections [{type=MovieSceneFloatSection, end_frame=35, start_frame=0, channels [None {type=MovieSceneScriptingFloatChannel, const [1,1]}]}]
      }]
    }
    LeftPanelCanvas {
     id=#2
     possessed_class=CanvasPanel
     tracks [
      {
       type=MovieSceneFloatTrack
       display_name=Render Opacity
       sections [{
         type=MovieSceneFloatSection
         end_frame=458
         start_frame=0
         channels [None {type=MovieSceneScriptingFloatChannel, keys [[0,0],[16,1],[442,1],[457,0]]}]
        }]
      }
      {
       type=MovieScene2DTransformTrack
       sections [{
         type=MovieScene2DTransformSection
         end_frame=458
         start_frame=0
         channels [
          Translation.X {type=MovieSceneScriptingFloatChannel, thinned_from=4, keys [[0,-308.7808],[16,0],[457,0]]}
          Translation.Y {type=MovieSceneScriptingFloatChannel, keys [[0,1088.9025],[16,0],[442,0],[457,-1091.6603]]}
          Angle {type=MovieSceneScriptingFloatChannel, const [0,2]}
          Scale.X {type=MovieSceneScriptingFloatChannel, const [1,2]}
          Scale.Y {type=MovieSceneScriptingFloatChannel, const [1,2]}
          Shear.X {type=MovieSceneScriptingFloatChannel, const [0,2]}
          Shear.Y {type=MovieSceneScriptingFloatChannel, const [0,2]}
         ]
        }]
      }
     ]
    }
    LeftPanel_Chr {
     id=#3
     possessed_class=Image
     tracks [
      {
       type=MovieScene2DTransformTrack
       sections [{
         type=MovieScene2DTransformSection
         end_frame=451
         start_frame=3
         channels [
          Translation.X {type=MovieSceneScriptingFloatChannel, default=0}
          Translation.Y {type=MovieSceneScriptingFloatChannel, default=0}
          Angle {type=MovieSceneScriptingFloatChannel, default=0, const [1.6796,1]}
          Scale.X {type=MovieSceneScriptingFloatChannel, keys [[3,1],[451,1.1]]}
          Scale.Y {type=MovieSceneScriptingFloatChannel, keys [[3,1],[451,1.1]]}
          Shear.X {type=MovieSceneScriptingFloatChannel, default=0}
          Shear.Y {type=MovieSceneScriptingFloatChannel, default=0}
         ]
        }]
      }
      {
       type=MovieSceneFloatTrack
       display_name=Render Opacity
       sections [{type=MovieSceneFloatSection, end_frame=35, start_frame=0, channels [None {type=MovieSceneScriptingFloatChannel, const [1,1]}]}]
      }
     ]
    }
   ]
   display_rate {numerator=60}
  }
 }
 RightPanelAnim {
  binding_count=3
  length_seconds=6.25
  object=&self:RightPanelAnim
  sequence {
   bindings [
    RightPanel_Chr {
     id=#4
     possessed_class=Image
     tracks [
      {
       type=MovieScene2DTransformTrack
       sections [{
         type=MovieScene2DTransformSection
         end_frame=370
         start_frame=2
         channels [
          Translation.X {type=MovieSceneScriptingFloatChannel, keys [[11,67.8925],[370,0]]}
          Translation.Y {type=MovieSceneScriptingFloatChannel, default=0, keys [[11,78.1947],[370,0]]}
          Angle {type=MovieSceneScriptingFloatChannel, default=0}
          Scale.X {type=MovieSceneScriptingFloatChannel, default=1}
          Scale.Y {type=MovieSceneScriptingFloatChannel, default=1}
          Shear.X {type=MovieSceneScriptingFloatChannel, default=0}
          Shear.Y {type=MovieSceneScriptingFloatChannel, default=0}
         ]
        }]
      }
      {
       type=MovieSceneFloatTrack
       display_name=Render Opacity
       sections [{type=MovieSceneFloatSection, end_frame=35, start_frame=0, channels [None {type=MovieSceneScriptingFloatChannel, const [1,1]}]}]
      }
     ]
    }
    RightPanelCanvas {
     id=#5
     possessed_class=CanvasPanel
     tracks [
      {
       type=MovieSceneFloatTrack
       display_name=Render Opacity
       sections [{
         type=MovieSceneFloatSection
         end_frame=375
         start_frame=1
         channels [None {type=MovieSceneScriptingFloatChannel, keys [[1,0],[16,1],[358,1],[374,0]]}]
        }]
      }
      {
       type=MovieScene2DTransformTrack
       sections [{
         type=MovieScene2DTransformSection
         end_frame=375
         start_frame=1
         channels [
          Translation.X {type=MovieSceneScriptingFloatChannel, thinned_from=4, keys [[1,353.58],[16,0],[374,0]]}
          Translation.Y {type=MovieSceneScriptingFloatChannel, keys [[1,-1081.0922],[16,0],[358,0],[374,1082.3345]]}
          Angle {type=MovieSceneScriptingFloatChannel, const [0,2]}
          Scale.X {type=MovieSceneScriptingFloatChannel, const [1,2]}
          Scale.Y {type=MovieSceneScriptingFloatChannel, const [1,2]}
          Shear.X {type=MovieSceneScriptingFloatChannel, const [0,2]}
          Shear.Y {type=MovieSceneScriptingFloatChannel, const [0,2]}
         ]
        }]
      }
     ]
    }
    RightPanel_BG {
     id=#6
     possessed_class=Image
     tracks [{
       type=MovieSceneFloatTrack
       display_name=Render Opacity
       sections [{type=MovieSceneFloatSection, end_frame=35, start_frame=0, channels [None {type=MovieSceneScriptingFloatChannel, const [1,1]}]}]
      }]
    }
   ]
   display_rate {numerator=60}
  }
 }
 NewAnimation {
  binding_count=1
  length_seconds=5
  object=&self:NewAnimation
  sequence {bindings [LeftPanel_BG {
     id=#7
     possessed_class=Image
     tracks [{
       type=MovieScene2DTransformTrack
       sections [{
         type=MovieScene2DTransformSection
         end_frame=0
         start_frame=0
         channels [
          Translation.X {type=MovieSceneScriptingFloatChannel, const [0,1]}
          Translation.Y {type=MovieSceneScriptingFloatChannel, const [0,1]}
          Angle {type=MovieSceneScriptingFloatChannel, const [0,1]}
          Scale.X {type=MovieSceneScriptingFloatChannel, const [1,1]}
          Scale.Y {type=MovieSceneScriptingFloatChannel, const [1,1]}
          Shear.X {type=MovieSceneScriptingFloatChannel, const [0,1]}
          Shear.Y {type=MovieSceneScriptingFloatChannel, const [0,1]}
         ]
        }]
      }]
    }]}
 }
]
graphs [EventGraph {nodes [
   Event Construct()
   > SetRenderOpacity(
    self=[
     Get RightPanel_Chr() -> (RightPanel_Chr:Image Object Reference)
     Get RightPanel_BG() -> (RightPanel_BG:Image Object Reference)
     Get LeftPanel_Chr() -> (LeftPanel_Chr:Image Object Reference)
     Get LeftPanel_BG() -> (LeftPanel_BG:Image Object Reference)
    ]
    InOpacity=0
   )
   Event HeroPanel_UnloadTextures() {is=title K2Node_Event}
   n1 = Make Array!(
    "[0]":Image Object Reference=Get LeftPanel_Chr()
    "[1]":Image Object Reference=Get LeftPanel_BG()
    "[2]":Image Object Reference=Get RightPanel_Chr()
    "[3]":Image Object Reference=Get RightPanel_BG()
   ) -> (Array:Array of Image Object References)
   > n2 = For Each Loop(Array:Array of Image Object References=@n1) -> (Array Element:Image Object Reference)
   CollectGarbage!(execute=@n2.Completed) -> (then)
   > PrintString!(execute, InString=UNLOADED HERO PANEL 1 TEXTURES, bPrintToScreen, bPrintToLog, TextColor=(R=0.9931,G=0.9931,B=1,A=1), Duration=3, Key) -> (then)
   Is Valid(exec=@n2.LoopBody, InputObject=@n2)
   > SetBrushFromTexture!(execute, self:Image Object Reference=@n2, Texture:Texture 2D Object Reference, bMatchSize:bool=true) -> (then)
   n3 = Event PlayPistolCanaryHeroPanelAnimation!() -> (OutputDelegate, then, HeroPanel:E_PistolCanary_HeroPanelSelect Enum=NewEnumerator0, PanelSide:E_PistolCanary_PanelSide Enum=NewEnumerator0) {is=title K2Node_Event}
   > PlayAnimation(InAnimation=Select!(
     NewEnumerator0:Widget Animation Object Reference=Get LeftPanelAnim() -> (LeftPanelAnim:Widget Animation Object Reference)
     NewEnumerator1:Widget Animation Object Reference=Get RightPanelAnim() -> (RightPanelAnim:Widget Animation Object Reference)
     NewEnumerator2:Widget Animation Object Reference
     NewEnumerator3:Widget Animation Object Reference
     NewEnumerator4:Widget Animation Object Reference
     NewEnumerator5:Widget Animation Object Reference
     Index:E_PistolCanary_HeroPanelSelect Enum=@n3.HeroPanel
    ) -> (ReturnValue:Widget Animation Object Reference))
   Event HeroPanel_LoadTextures() {is=title K2Node_Event}
   > n4 = For Each Loop(Array=@n1)
   n5 = IsMobilePlatform!() -> (ReturnValue:bool)
   n6 = Get Textures() -> (Textures:Array of Texture 2D Soft Object References)
   n7 = Get TexturesLow() -> (TexturesLow:Array of Texture 2D Soft Object References)
   > SetBrushFromSoftTexture!(
    execute
    self:Image Object Reference=@n4
    SoftTexture:Texture 2D Soft Object Reference="Get (a copy)"(
     Array:Array of Texture 2D Soft Object References=Select!(Option 0:Array of Texture 2D Soft Object References=@n6, Option 1:Array of Texture 2D Soft Object References=@n7, Index:bool=@n5) -> (ReturnValue:Array of Texture 2D Soft Object References)
     Dimension 1=@n4.Array Index
    ) -> (Output:Texture 2D Soft Object Reference)
    bMatchSize:bool=false
   ) -> (then)
   Event PreConstruct()
   Event Tick()
  ]}]
variables [
 Textures {
  type="array<Texture2D>"
  default [
   /PistolCanary/Layout/Texture/ComicPanels/HeroPanels/Hero1/FN_Comic_Overlay_A1_character
   /PistolCanary/Layout/Texture/ComicPanels/HeroPanels/Hero1/FN_Comic_Overlay_A1_background
   /PistolCanary/Layout/Texture/ComicPanels/HeroPanels/Hero1/FN_Comic_Overlay_A2_character
   /PistolCanary/Layout/Texture/ComicPanels/HeroPanels/Hero1/FN_Comic_Overlay_A2_background
  ]
 }
 TexturesLow {
  type="array<Texture2D>"
  default [
   /PistolCanary/Layout/Texture/ComicPanels/HeroPanels/Hero1/LowRes/FN_Comic_Overlay_A1_character_low
   /PistolCanary/Layout/Texture/ComicPanels/HeroPanels/Hero1/LowRes/FN_Comic_Overlay_A1_background_low
   /PistolCanary/Layout/Texture/ComicPanels/HeroPanels/Hero1/LowRes/FN_Comic_Overlay_A2_character_low
   /PistolCanary/Layout/Texture/ComicPanels/HeroPanels/Hero1/LowRes/FN_Comic_Overlay_A2_background_low
  ]
 }
]
widgets {
 info {inheritedWidgetCount=0, parentClass=&UserWidget, rootWidgetClass=&CanvasPanel, widgetCount=7}
 widgets [
  CanvasPanel_163 {widgetClassPath=&CanvasPanel, widget=CanvasPanel_163, properties {clipping=ClipToBounds}}
  LeftPanelCanvas {
   widgetClassPath=&CanvasPanel
   parent=CanvasPanel_163
   widget=LeftPanelCanvas
   properties {clipping=ClipToBounds}
   slot {
    class=CanvasPanelSlot
    properties {bAutoSize=true, layoutData {alignment=<0 0>, anchors {maximum=<0 0.5>, minimum=<0 0.5>}, offsets=<ltrb 0 -540.5405 100 30>}}
   }
  }
  LeftPanel_BG {
   widgetClassPath=&Image
   bIsVariable=true
   parent=LeftPanelCanvas
   widget=LeftPanel_BG
   properties {
    clipping=ClipToBounds
    visibility=HitTestInvisible
    brush {
     bIsDynamicallyLoaded=false
     drawAs=Image
     imageType=FullColor
     mirroring=NoMirror
     resourceName=None
     resourceObject=&FN_Comic_Overlay_A1_background
     tiling=NoTile
     imageSize=<312 512>
     margin=<ltrb 0 0 0 0>
     outlineSettings {
      bUseBrushTransparency=false
      roundingType=HalfHeightRadius
      width=0
      color {colorUseRule=UseColor_Specified, specifiedColor=<rgba 0 0 0 0>}
      cornerRadii=<0 0 0 0>
     }
     tintColor {colorUseRule=UseColor_Specified, specifiedColor=<rgba 1 1 1 1>}
     uVRegion {bIsValid=false, max=<0 0>, min=<0 0>}
    }
   }
   slot {class=CanvasPanelSlot, properties {layoutData {alignment=<0 0>, anchors {maximum=<0 0>, minimum=<0 0>}, offsets=<ltrb 0 0 658.125 1080>}}}
  }
  LeftPanel_Chr {
   widgetClassPath=&Image
   bIsVariable=true
   parent=LeftPanelCanvas
   widget=LeftPanel_Chr
   properties {
    clipping=ClipToBounds
    visibility=HitTestInvisible
    brush {
     bIsDynamicallyLoaded=false
     drawAs=Image
     imageType=FullColor
     mirroring=NoMirror
     resourceName=None
     resourceObject=&FN_Comic_Overlay_A1_character
     tiling=NoTile
     imageSize=<844 1096>
     margin=<ltrb 0 0 0 0>
     outlineSettings {
      bUseBrushTransparency=false
      roundingType=HalfHeightRadius
      width=0
      color {colorUseRule=UseColor_Specified, specifiedColor=<rgba 0 0 0 0>}
      cornerRadii=<0 0 0 0>
     }
     tintColor {colorUseRule=UseColor_Specified, specifiedColor=<rgba 1 1 1 1>}
     uVRegion {bIsValid=false, max=<0 0>, min=<0 0>}
    }
   }
   slot {
    class=CanvasPanelSlot
    properties {
     zOrder=1
     layoutData {alignment=<0 0>, anchors {maximum=<0.2319 0.1721>, minimum=<0.2319 0.1721>}, offsets=<ltrb -195.7513 -188.633 844 1096>}
    }
   }
  }
  RightPanelCanvas {
   widgetClassPath=&CanvasPanel
   parent=CanvasPanel_163
   widget=RightPanelCanvas
   slot {
    class=CanvasPanelSlot
    properties {bAutoSize=true, layoutData {alignment=<0 0>, anchors {maximum=<1 0.5>, minimum=<1 0.5>}, offsets=<ltrb -723.4156 -540 100 30>}}
   }
  }
  RightPanel_BG {
   widgetClassPath=&Image
   bIsVariable=true
   parent=RightPanelCanvas
   widget=RightPanel_BG
   properties {
    visibility=HitTestInvisible
    brush {
     bIsDynamicallyLoaded=false
     drawAs=Image
     imageType=FullColor
     mirroring=NoMirror
     resourceName=None
     resourceObject=&FN_Comic_Overlay_A2_background
     tiling=NoTile
     imageSize=<336 512>
     margin=<ltrb 0 0 0 0>
     outlineSettings {
      bUseBrushTransparency=false
      roundingType=HalfHeightRadius
      width=0
      color {colorUseRule=UseColor_Specified, specifiedColor=<rgba 0 0 0 0>}
      cornerRadii=<0 0 0 0>
     }
     tintColor {colorUseRule=UseColor_Specified, specifiedColor=<rgba 1 1 1 1>}
     uVRegion {bIsValid=false, max=<0 0>, min=<0 0>}
    }
   }
   slot {class=CanvasPanelSlot, properties {layoutData {alignment=<0 0>, anchors {maximum=<0 0>, minimum=<0 0>}, offsets=<ltrb 0 0 723.4156 1080>}}}
  }
  RightPanel_Chr {
   widgetClassPath=&Image
   bIsVariable=true
   parent=RightPanelCanvas
   widget=RightPanel_Chr
   properties {
    visibility=HitTestInvisible
    brush {
     bIsDynamicallyLoaded=false
     drawAs=Image
     imageType=FullColor
     mirroring=NoMirror
     resourceName=None
     resourceObject=&FN_Comic_Overlay_A2_character
     tiling=NoTile
     imageSize=<700 1064>
     margin=<ltrb 0 0 0 0>
     outlineSettings {
      bUseBrushTransparency=false
      roundingType=HalfHeightRadius
      width=0
      color {colorUseRule=UseColor_Specified, specifiedColor=<rgba 0 0 0 0>}
      cornerRadii=<0 0 0 0>
     }
     tintColor {colorUseRule=UseColor_Specified, specifiedColor=<rgba 1 1 1 1>}
     uVRegion {bIsValid=false, max=<0 0>, min=<0 0>}
    }
   }
   slot {
    class=CanvasPanelSlot
    properties {zOrder=1, layoutData {alignment=<0 0>, anchors {maximum=<0 0>, minimum=<0 0>}, offsets=<ltrb 0 0 723.4156 1080>}}
   }
  }
 ]
}
