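# PistolCanary comic overlay widget for hero panel 2, with layered left and right panels that slide in, drift and slide out.
# Construct hides every image; the play event picks LeftPanelAnim or RightPanelAnim by HeroPanel enum, with sound-effect lettering popping in on the right.
# Load fills images from index-matched soft texture arrays, using low-res versions on mobile.
# Unload clears the brushes, then forces garbage collection and prints a debug message.
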WidgetBlueprint /PistolCanary/Layout/BP/ComicPanels/HeroPanels/PistolCanary_HeroPanel2
animations [
 LeftPanelAnim {
  binding_count=4
  length_seconds=7.9667
  object=&self:LeftPanelAnim
  sequence {
   bindings [
    LeftPanelCanvas {
     id=#1
     possessed_class=CanvasPanel
     tracks [
      {
       type=MovieSceneFloatTrack
       display_name=Render Opacity
       sections [{
         type=MovieSceneFloatSection
         end_frame=478
         start_frame=0
         channels [None {type=MovieSceneScriptingFloatChannel, thinned_from=4, keys [[0,1],[462,1],[477,0]]}]
        }]
      }
      {
       type=MovieScene2DTransformTrack
       sections [{
         type=MovieScene2DTransformSection
         end_frame=478
         start_frame=0
         channels [
          Translation.X {type=MovieSceneScriptingFloatChannel, thinned_from=4, keys [[0,-341.2048],[16,0],[477,0]]}
          Translation.Y {type=MovieSceneScriptingFloatChannel, keys [[0,-1082.9917],[16,0],[462,0],[477,1085.1641]]}
          Angle {type=MovieSceneScriptingFloatChannel, const [0,3]}
          Scale.X {type=MovieSceneScriptingFloatChannel, const [1,3]}
          Scale.Y {type=MovieSceneScriptingFloatChannel, const [1,3]}
          Shear.X {type=MovieSceneScriptingFloatChannel, const [0,3]}
          Shear.Y {type=MovieSceneScriptingFloatChannel, const [0,3]}
         ]
        }]
      }
     ]
    }
    LeftPanel_Chr {
     id=#2
     possessed_class=Image
     tracks [
      {
       type=MovieSceneFloatTrack
       display_name=Render Opacity
       sections [{type=MovieSceneFloatSection, end_frame=1, start_frame=0, channels [None {type=MovieSceneScriptingFloatChannel, const [1,1]}]}]
      }
      {
       type=MovieScene2DTransformTrack
       sections [{
         type=MovieScene2DTransformSection
         end_frame=473
         start_frame=4
         channels [
          Translation.X {type=MovieSceneScriptingFloatChannel, default=0}
          Translation.Y {type=MovieSceneScriptingFloatChannel, keys [[4,0],[473,-99.931]]}
          Angle {type=MovieSceneScriptingFloatChannel, default=0}
          Scale.X {type=MovieSceneScriptingFloatChannel, default=1.1}
          Scale.Y {type=MovieSceneScriptingFloatChannel, default=1.1}
          Shear.X {type=MovieSceneScriptingFloatChannel, default=0}
          Shear.Y {type=MovieSceneScriptingFloatChannel, default=0}
         ]
        }]
      }
     ]
    }
    LeftPanel_BG {
     id=#3
     possessed_class=Image
     tracks [{
       type=MovieSceneFloatTrack
       display_name=Render Opacity
       sections [{type=MovieSceneFloatSection, end_frame=1, start_frame=0, channels [None {type=MovieSceneScriptingFloatChannel, const [1,1]}]}]
      }]
    }
    LeftPanel_Flare {
     id=#4
     possessed_class=Image
     tracks [
      {
       type=MovieSceneFloatTrack
       display_name=Render Opacity
       sections [{type=MovieSceneFloatSection, end_frame=1, start_frame=0, channels [None {type=MovieSceneScriptingFloatChannel, const [1,1]}]}]
      }
      {
       type=MovieScene2DTransformTrack
       sections [{
         type=MovieScene2DTransformSection
         end_frame=472
         start_frame=5
         channels [
          Translation.X {type=MovieSceneScriptingFloatChannel, default=0}
          Translation.Y {type=MovieSceneScriptingFloatChannel, keys [[5,0],[462,-20.2522]]}
          Angle {type=MovieSceneScriptingFloatChannel, default=0}
          Scale.X {type=MovieSceneScriptingFloatChannel, const [1.5,1]}
          Scale.Y {type=MovieSceneScriptingFloatChannel, const [1.5,1]}
          Shear.X {type=MovieSceneScriptingFloatChannel, const [0,1]}
          Shear.Y {type=MovieSceneScriptingFloatChannel, const [0,1]}
         ]
        }]
      }
     ]
    }
   ]
   display_rate {numerator=60}
  }
 }
 RightPanelAnim {
  binding_count=10
  length_seconds=6.6333
  object=&self:RightPanelAnim
  sequence {
   bindings [
    RightPanel_Bang {
     id=#5
     possessed_class=Image
     tracks [
      {
       type=MovieSceneFloatTrack
       display_name=Render Opacity
       sections [{type=MovieSceneFloatSection, end_frame=125, start_frame=0, channels [None {type=MovieSceneScriptingFloatChannel, keys [[0,0],[125,1]]}]}]
      }
      {
       type=MovieScene2DTransformTrack
       sections [{
         type=MovieScene2DTransformSection
         end_frame=141
         start_frame=67
         channels [
          Translation.X {type=MovieSceneScriptingFloatChannel, default=0}
          Translation.Y {type=MovieSceneScriptingFloatChannel, default=0}
          Angle {type=MovieSceneScriptingFloatChannel, default=0}
          Scale.X {type=MovieSceneScriptingFloatChannel, keys [[125,0],[141,1]]}
          Scale.Y {type=MovieSceneScriptingFloatChannel, keys [[125,0],[141,1]]}
          Shear.X {type=MovieSceneScriptingFloatChannel, default=0}
          Shear.Y {type=MovieSceneScriptingFloatChannel, default=0}
         ]
        }]
      }
     ]
    }
    RightPanel_BG {
     id=#6
     possessed_class=Image
     tracks [{
       type=MovieSceneFloatTrack
       display_name=Render Opacity
       sections [{type=MovieSceneFloatSection, end_frame=1, start_frame=0, channels [None {type=MovieSceneScriptingFloatChannel, const [1,1]}]}]
      }]
    }
    RightPanelCanvas {
     id=#7
     possessed_class=CanvasPanel
     tracks [
      {
       type=MovieSceneFloatTrack
       display_name=Render Opacity
       sections [{
         type=MovieSceneFloatSection
         end_frame=395
         start_frame=0
         channels [None {type=MovieSceneScriptingFloatChannel, keys [[0,0],[15,1],[378,1],[394,0]]}]
        }]
      }
      {
       type=MovieScene2DTransformTrack
       sections [{
         type=MovieScene2DTransformSection
         end_frame=395
         start_frame=0
         channels [
          Translation.X {type=MovieSceneScriptingFloatChannel, thinned_from=4, keys [[0,353.58],[15,0],[394,0]]}
          Translation.Y {type=MovieSceneScriptingFloatChannel, keys [[0,1081.635],[15,0],[378,0],[394,-1083.9471]]}
          Angle {type=MovieSceneScriptingFloatChannel, const [0,4]}
          Scale.X {type=MovieSceneScriptingFloatChannel, const [1,4]}
          Scale.Y {type=MovieSceneScriptingFloatChannel, const [1,4]}
          Shear.X {type=MovieSceneScriptingFloatChannel, const [0,4]}
          Shear.Y {type=MovieSceneScriptingFloatChannel, const [0,4]}
         ]
        }]
      }
     ]
    }
    RightPanel_Boom1 {
     id=#8
     possessed_class=Image
     tracks [
      {
       type=MovieSceneFloatTrack
       display_name=Render Opacity
       sections [{type=MovieSceneFloatSection, end_frame=214, start_frame=0, channels [None {type=MovieSceneScriptingFloatChannel, keys [[0,0],[214,1]]}]}]
      }
      {
       type=MovieScene2DTransformTrack
       sections [{
         type=MovieScene2DTransformSection
         end_frame=231
         start_frame=156
         channels [
          Translation.X {type=MovieSceneScriptingFloatChannel, default=0}
          Translation.Y {type=MovieSceneScriptingFloatChannel, default=0}
          Angle {type=MovieSceneScriptingFloatChannel, default=0}
          Scale.X {type=MovieSceneScriptingFloatChannel, keys [[214,0],[231,1]]}
          Scale.Y {type=MovieSceneScriptingFloatChannel, keys [[214,0],[231,1]]}
          Shear.X {type=MovieSceneScriptingFloatChannel, default=0}
          Shear.Y {type=MovieSceneScriptingFloatChannel, default=0}
         ]
        }]
      }
     ]
    }
    RightPanel_Ping2 {
     id=#9
     possessed_class=Image
     tracks [
      {
       type=MovieSceneFloatTrack
       display_name=Render Opacity
       sections [{type=MovieSceneFloatSection, end_frame=69, start_frame=0, channels [None {type=MovieSceneScriptingFloatChannel, keys [[0,0],[69,1]]}]}]
      }
      {
       type=MovieScene2DTransformTrack
       sections [{
         type=MovieScene2DTransformSection
         end_frame=273
         start_frame=69
         channels [
          Translation.X {type=MovieSceneScriptingFloatChannel, default=0}
          Translation.Y {type=MovieSceneScriptingFloatChannel, default=0}
          Angle {type=MovieSceneScriptingFloatChannel, default=0}
          Scale.X {type=MovieSceneScriptingFloatChannel, keys [[69,0],[85,1]]}
          Scale.Y {type=MovieSceneScriptingFloatChannel, keys [[69,0],[85,1]]}
          Shear.X {type=MovieSceneScriptingFloatChannel, default=0}
          Shear.Y {type=MovieSceneScriptingFloatChannel, default=0}
         ]
        }]
      }
     ]
    }
    RightPanel_Ping1 {
     id=#10
     possessed_class=Image
     tracks [
      {
       type=MovieSceneFloatTrack
       display_name=Render Opacity
       sections [{type=MovieSceneFloatSection, end_frame=171, start_frame=0, channels [None {type=MovieSceneScriptingFloatChannel, keys [[0,0],[171,1]]}]}]
      }
      {
       type=MovieScene2DTransformTrack
       sections [{
         type=MovieScene2DTransformSection
         end_frame=185
         start_frame=104
         channels [
          Translation.X {type=MovieSceneScriptingFloatChannel, default=0}
          Translation.Y {type=MovieSceneScriptingFloatChannel, default=0}
          Angle {type=MovieSceneScriptingFloatChannel, default=0}
          Scale.X {type=MovieSceneScriptingFloatChannel, keys [[171,0],[185,1]]}
          Scale.Y {type=MovieSceneScriptingFloatChannel, keys [[171,0],[185,1]]}
          Shear.X {type=MovieSceneScriptingFloatChannel, default=0}
          Shear.Y {type=MovieSceneScriptingFloatChannel, default=0}
         ]
        }]
      }
     ]
    }
    RightPanel_Boom2 {
     id=#11
     possessed_class=Image
     tracks [
      {
       type=MovieSceneFloatTrack
       display_name=Render Opacity
       sections [{type=MovieSceneFloatSection, end_frame=25, start_frame=0, channels [None {type=MovieSceneScriptingFloatChannel, keys [[0,0],[25,1]]}]}]
      }
      {
       type=MovieScene2DTransformTrack
       sections [{
         type=MovieScene2DTransformSection
         end_frame=113
         start_frame=0
         channels [
          Translation.X {type=MovieSceneScriptingFloatChannel, default=0}
          Translation.Y {type=MovieSceneScriptingFloatChannel, default=0}
          Angle {type=MovieSceneScriptingFloatChannel, default=0}
          Scale.X {type=MovieSceneScriptingFloatChannel, keys [[25,0],[37,1]]}
          Scale.Y {type=MovieSceneScriptingFloatChannel, keys [[25,0],[37,1]]}
          Shear.X {type=MovieSceneScriptingFloatChannel, default=0}
          Shear.Y {type=MovieSceneScriptingFloatChannel, default=0}
         ]
        }]
      }
     ]
    }
    RightPanel_Flames {
     id=#12
     possessed_class=Image
     tracks [
      {
       type=MovieSceneFloatTrack
       display_name=Render Opacity
       sections [{type=MovieSceneFloatSection, end_frame=1, start_frame=0, channels [None {type=MovieSceneScriptingFloatChannel, const [1,1]}]}]
      }
      {
       type=MovieScene2DTransformTrack
       sections [{
         type=MovieScene2DTransformSection
         end_frame=386
         start_frame=4
         channels [
          Translation.X {type=MovieSceneScriptingFloatChannel, default=0}
          Translation.Y {type=MovieSceneScriptingFloatChannel, keys [[4,33.2125],[386,0]]}
          Angle {type=MovieSceneScriptingFloatChannel, default=0}
          Scale.X {type=MovieSceneScriptingFloatChannel, default=1.5}
          Scale.Y {type=MovieSceneScriptingFloatChannel, default=1.5}
          Shear.X {type=MovieSceneScriptingFloatChannel, default=0}
          Shear.Y {type=MovieSceneScriptingFloatChannel, default=0}
         ]
        }]
      }
     ]
    }
    RightPanel_Chr {
     id=#13
     possessed_class=Image
     tracks [
      {
       type=MovieSceneFloatTrack
       display_name=Render Opacity
       sections [{type=MovieSceneFloatSection, end_frame=1, start_frame=0, channels [None {type=MovieSceneScriptingFloatChannel, const [1,1]}]}]
      }
      {
       type=MovieScene2DTransformTrack
       sections [{
         type=MovieScene2DTransformSection
         end_frame=398
         start_frame=1
         channels [
          Translation.X {type=MovieSceneScriptingFloatChannel, keys [[1,76.6993],[390,10.8279]]}
          Translation.Y {type=MovieSceneScriptingFloatChannel, keys [[1,-82.6081],[390,-21.9355]]}
          Angle {type=MovieSceneScriptingFloatChannel, default=0}
          Scale.X {type=MovieSceneScriptingFloatChannel, default=0.7}
          Scale.Y {type=MovieSceneScriptingFloatChannel, default=0.7}
          Shear.X {type=MovieSceneScriptingFloatChannel, default=0}
          Shear.Y {type=MovieSceneScriptingFloatChannel, default=0}
         ]
        }]
      }
     ]
    }
    RightPanel_House {
     id=#14
     possessed_class=Image
     tracks [{
       type=MovieSceneFloatTrack
       display_name=Render Opacity
       sections [{type=MovieSceneFloatSection, end_frame=1, start_frame=0, channels [None {type=MovieSceneScriptingFloatChannel, const [1,1]}]}]
      }]
    }
   ]
   display_rate {numerator=60}
  }
 }
]
graphs [EventGraph {nodes [
   Event Construct()
   > SetRenderOpacity(
    self=[
     Get LeftPanel_BG() -> (LeftPanel_BG:Image Object Reference)
     Get LeftPanel_Flare() -> (LeftPanel_Flare:Image Object Reference)
     Get LeftPanel_Chr() -> (LeftPanel_Chr:Image Object Reference)
     Get RightPanel_BG() -> (RightPanel_BG:Image Object Reference)
     Get RightPanel_Chr() -> (RightPanel_Chr:Image Object Reference)
     Get RightPanel_House() -> (RightPanel_House:Image Object Reference)
     Get RightPanel_Flames() -> (RightPanel_Flames:Image Object Reference)
     Get RightPanel_Bang() -> (RightPanel_Bang:Image Object Reference)
     Get RightPanel_Boom1() -> (RightPanel_Boom1:Image Object Reference)
     Get RightPanel_Boom2() -> (RightPanel_Boom2:Image Object Reference)
     Get RightPanel_Ping1() -> (RightPanel_Ping1:Image Object Reference)
     Get RightPanel_Ping2() -> (RightPanel_Ping2:Image Object Reference)
    ]
    InOpacity=0
   )
   Event HeroPanel_UnloadTextures() {is=title K2Node_Event}
   n1 = Make Array!(
    "[0]":Image Object Reference=Get LeftPanel_BG()
    "[1]":Image Object Reference=Get LeftPanel_Chr()
    "[2]":Image Object Reference=Get LeftPanel_Flare()
    "[3]":Image Object Reference=Get RightPanel_BG()
    "[4]":Image Object Reference=Get RightPanel_Chr()
    "[5]":Image Object Reference=Get RightPanel_House()
    "[6]":Image Object Reference=Get RightPanel_Flames()
    "[7]":Image Object Reference=Get RightPanel_Bang()
    "[8]":Image Object Reference=Get RightPanel_Boom1()
    "[9]":Image Object Reference=Get RightPanel_Boom2()
    "[10]":Image Object Reference=Get RightPanel_Ping1()
    "[11]":Image Object Reference=Get RightPanel_Ping2()
   ) -> (Array:Array of Image Object References)
   > n2 = For Each Loop(Array:Array of Image Object References=@n1) -> (Array Element:Image Object Reference)
   CollectGarbage!(execute=@n2.Completed) -> (then)
   > PrintString!(execute, InString=COMIC PANEL 2 UNLOADED, bPrintToScreen, bPrintToLog, TextColor=(R=1,G=0.9595,B=0.9817,A=1), Duration=3, Key) -> (then)
   Is Valid(exec=@n2.LoopBody, InputObject=@n2)
   > SetBrushFromTexture!(execute, self:Image Object Reference=@n2, Texture:Texture 2D Object Reference, bMatchSize:bool=true) -> (then)
   n3 = Event PlayPistolCanaryHeroPanelAnimation!() -> (OutputDelegate, then, HeroPanel:E_PistolCanary_HeroPanelSelect Enum=NewEnumerator0, PanelSide:E_PistolCanary_PanelSide Enum=NewEnumerator0) {is=title K2Node_Event}
   > PlayAnimation(InAnimation=Select!(
     NewEnumerator0:Widget Animation Object Reference
     NewEnumerator1:Widget Animation Object Reference
     NewEnumerator2:Widget Animation Object Reference=Get LeftPanelAnim() -> (LeftPanelAnim:Widget Animation Object Reference)
     NewEnumerator3:Widget Animation Object Reference=Get RightPanelAnim() -> (RightPanelAnim:Widget Animation Object Reference)
     NewEnumerator4:Widget Animation Object Reference
     NewEnumerator5:Widget Animation Object Reference
     Index:E_PistolCanary_HeroPanelSelect Enum=@n3.HeroPanel
    ) -> (ReturnValue:Widget Animation Object Reference))
   Event HeroPanel_LoadTextures() {is=title K2Node_Event}
   > n4 = For Each Loop(Array=@n1)
   n5 = IsMobilePlatform!() -> (ReturnValue:bool)
   n6 = Get TexturesLow() -> (TexturesLow:Array of Texture 2D Soft Object References)
   n7 = Get Textures() -> (Textures:Array of Texture 2D Soft Object References)
   > SetBrushFromSoftTexture!(
    execute
    self:Image Object Reference=@n4
    SoftTexture:Texture 2D Soft Object Reference="Get (a copy)"(
     Array:Array of Texture 2D Soft Object References=Select!(Option 0:Array of Texture 2D Soft Object References=@n7, Option 1:Array of Texture 2D Soft Object References=@n6, Index:bool=@n5) -> (ReturnValue:Array of Texture 2D Soft Object References)
     Dimension 1=@n4.Array Index
    ) -> (Output:Texture 2D Soft Object Reference)
    bMatchSize:bool=false
   ) -> (then)
   Event PreConstruct()
   Event Tick()
  ]}]
variables [
 Textures {
  type="array<Texture2D>"
  default [
   /PistolCanary/Layout/Texture/ComicPanels/HeroPanels/Hero2/FN_Comic_Overlay_B1_background
   /PistolCanary/Layout/Texture/ComicPanels/HeroPanels/Hero2/FN_Comic_Overlay_B1_character
   /PistolCanary/Layout/Texture/ComicPanels/HeroPanels/Hero2/FN_Comic_Overlay_B1_flares
   /PistolCanary/Layout/Texture/ComicPanels/HeroPanels/Hero2/FN_Comic_Overlay_B2_background
   /PistolCanary/Layout/Texture/ComicPanels/HeroPanels/Hero2/FN_Comic_Overlay_B2_character
   /PistolCanary/Layout/Texture/ComicPanels/HeroPanels/Hero2/FN_Comic_Overlay_B2_house
   /PistolCanary/Layout/Texture/ComicPanels/HeroPanels/Hero2/FN_Comic_Overlay_B2_flames
   /PistolCanary/Layout/Texture/ComicPanels/HeroPanels/Hero2/FN_Comic_Overlay_B2_bang
   /PistolCanary/Layout/Texture/ComicPanels/HeroPanels/Hero2/FN_Comic_Overlay_B2_lower_boom
   /PistolCanary/Layout/Texture/ComicPanels/HeroPanels/Hero2/FN_Comic_Overlay_B2_upper_boom
   /PistolCanary/Layout/Texture/ComicPanels/HeroPanels/Hero2/FN_Comic_Overlay_B2_pingg
   /PistolCanary/Layout/Texture/ComicPanels/HeroPanels/Hero2/FN_Comic_Overlay_B2_pinggg
  ]
 }
 TexturesLow {
  type="array<Texture2D>"
  default [
   /PistolCanary/Layout/Texture/ComicPanels/HeroPanels/Hero2/LowRes/FN_Comic_Overlay_B1_background_low
   /PistolCanary/Layout/Texture/ComicPanels/HeroPanels/Hero2/LowRes/FN_Comic_Overlay_B1_character_low
   /PistolCanary/Layout/Texture/ComicPanels/HeroPanels/Hero2/LowRes/FN_Comic_Overlay_B1_flares_low
   /PistolCanary/Layout/Texture/ComicPanels/HeroPanels/Hero2/LowRes/FN_Comic_Overlay_B2_background_low
   /PistolCanary/Layout/Texture/ComicPanels/HeroPanels/Hero2/LowRes/FN_Comic_Overlay_B2_character_low
   /PistolCanary/Layout/Texture/ComicPanels/HeroPanels/Hero2/LowRes/FN_Comic_Overlay_B2_house_low
   /PistolCanary/Layout/Texture/ComicPanels/HeroPanels/Hero2/LowRes/FN_Comic_Overlay_B2_flames_low
   /PistolCanary/Layout/Texture/ComicPanels/HeroPanels/Hero2/LowRes/FN_Comic_Overlay_B2_bang_low
   /PistolCanary/Layout/Texture/ComicPanels/HeroPanels/Hero2/LowRes/FN_Comic_Overlay_B2_lower_boom_low
   /PistolCanary/Layout/Texture/ComicPanels/HeroPanels/Hero2/LowRes/FN_Comic_Overlay_B2_upper_boom_low
   /PistolCanary/Layout/Texture/ComicPanels/HeroPanels/Hero2/LowRes/FN_Comic_Overlay_B2_pingg_low
   /PistolCanary/Layout/Texture/ComicPanels/HeroPanels/Hero2/LowRes/FN_Comic_Overlay_B2_pinggg_low
  ]
 }
]
widgets {
 info {inheritedWidgetCount=0, parentClass=&UserWidget, rootWidgetClass=&CanvasPanel, widgetCount=15}
 widgets [
  CanvasPanel_163 {widgetClassPath=&CanvasPanel, widget=CanvasPanel_163, properties {clipping=ClipToBounds, visibility=HitTestInvisible}}
  LeftPanelCanvas {
   widgetClassPath=&CanvasPanel
   parent=CanvasPanel_163
   widget=LeftPanelCanvas
   properties {clipping=ClipToBounds, visibility=HitTestInvisible}
   slot {
    class=CanvasPanelSlot
    properties {layoutData {alignment=<0 0>, anchors {maximum=<0 0.5>, minimum=<0 0.5>}, offsets=<ltrb 0 -540.5405 653.9063 1080>}}
   }
  }
  LeftPanel_BG {
   widgetClassPath=&Image
   bIsVariable=true
   parent=LeftPanelCanvas
   widget=LeftPanel_BG
   properties {brush {
     bIsDynamicallyLoaded=false
     drawAs=Image
     imageType=FullColor
     mirroring=NoMirror
     resourceName=None
     resourceObject=null
     tiling=NoTile
     imageSize=<312 512>
     margin=<ltrb 0 0 0 0>
     outlineSettings {
      bUseBrushTransparency=false
      roundingType=HalfHeightRadius
      width=0
      color {colorUseRule=UseColor_Specified, specifiedColor=<rgba 0 0 0 0>}
      cornerRadii=<0 0 0 0>
     }
     tintColor {colorUseRule=UseColor_Specified, specifiedColor=<rgba 1 1 1 1>}
     uVRegion {bIsValid=false, max=<0 0>, min=<0 0>}
    }}
   slot {class=CanvasPanelSlot, properties {layoutData {alignment=<0 0>, anchors {maximum=<0 0>, minimum=<0 0>}, offsets=<ltrb 0 0 653.9063 1080>}}}
  }
  LeftPanel_Flare {
   widgetClassPath=&Image
   bIsVariable=true
   parent=LeftPanelCanvas
   widget=LeftPanel_Flare
   properties {
    brush {
     bIsDynamicallyLoaded=false
     drawAs=Image
     imageType=FullColor
     mirroring=NoMirror
     resourceName=None
     resourceObject=null
     tiling=NoTile
     imageSize=<300 380>
     margin=<ltrb 0 0 0 0>
     outlineSettings {
      bUseBrushTransparency=false
      roundingType=HalfHeightRadius
      width=0
      color {colorUseRule=UseColor_Specified, specifiedColor=<rgba 0 0 0 0>}
      cornerRadii=<0 0 0 0>
     }
     tintColor {colorUseRule=UseColor_Specified, specifiedColor=<rgba 1 1 1 1>}
     uVRegion {bIsValid=false, max=<0 0>, min=<0 0>}
    }
    renderTransform {angle=0, scale=<1.1 1.1>, shear=<0 0>, translation=<0 0>}
   }
   slot {
    class=CanvasPanelSlot
    properties {layoutData {alignment=<0 0>, anchors {maximum=<0 0>, minimum=<0 0>}, offsets=<ltrb 92.1217 177.5993 300 380>}}
   }
  }
  LeftPanel_Chr {
   widgetClassPath=&Image
   bIsVariable=true
   parent=LeftPanelCanvas
   widget=LeftPanel_Chr
   properties {
    brush {
     bIsDynamicallyLoaded=false
     drawAs=Image
     imageType=FullColor
     mirroring=NoMirror
     resourceName=None
     resourceObject=null
     tiling=NoTile
     imageSize=<600 1080>
     margin=<ltrb 0 0 0 0>
     outlineSettings {
      bUseBrushTransparency=false
      roundingType=HalfHeightRadius
      width=0
      color {colorUseRule=UseColor_Specified, specifiedColor=<rgba 0 0 0 0>}
      cornerRadii=<0 0 0 0>
     }
     tintColor {colorUseRule=UseColor_Specified, specifiedColor=<rgba 1 1 1 1>}
     uVRegion {bIsValid=false, max=<0 0>, min=<0 0>}
    }
    renderTransform {angle=0, scale=<1.1 1.1>, shear=<0 0>, translation=<0 0>}
   }
   slot {
    class=CanvasPanelSlot
    properties {layoutData {alignment=<0 0>, anchors {maximum=<0 0>, minimum=<0 0>}, offsets=<ltrb -63.487 54.4453 653.9063 1080>}}
   }
  }
  RightPanelCanvas {
   widgetClassPath=&CanvasPanel
   parent=CanvasPanel_163
   widget=RightPanelCanvas
   properties {clipping=ClipToBounds, visibility=HitTestInvisible}
   slot {
    class=CanvasPanelSlot
    properties {layoutData {alignment=<0 0>, anchors {maximum=<1 0.5>, minimum=<1 0.5>}, offsets=<ltrb -797.5875 -540 797.5875 1080>}}
   }
  }
  RightPanel_BG {
   widgetClassPath=&Image
   bIsVariable=true
   parent=RightPanelCanvas
   widget=RightPanel_BG
   properties {brush {
     bIsDynamicallyLoaded=false
     drawAs=Image
     imageType=FullColor
     mirroring=NoMirror
     resourceName=None
     resourceObject=null
     tiling=NoTile
     imageSize=<332 516>
     margin=<ltrb 0 0 0 0>
     outlineSettings {
      bUseBrushTransparency=false
      roundingType=HalfHeightRadius
      width=0
      color {colorUseRule=UseColor_Specified, specifiedColor=<rgba 0 0 0 0>}
      cornerRadii=<0 0 0 0>
     }
     tintColor {colorUseRule=UseColor_Specified, specifiedColor=<rgba 1 1 1 1>}
     uVRegion {bIsValid=false, max=<0 0>, min=<0 0>}
    }}
   slot {class=CanvasPanelSlot, properties {layoutData {alignment=<0 0>, anchors {maximum=<0 0>, minimum=<0 0>}, offsets=<ltrb 100 0 697.5875 1080>}}}
  }
  RightPanel_Chr {
   widgetClassPath=&Image
   bIsVariable=true
   parent=RightPanelCanvas
   widget=RightPanel_Chr
   properties {
    brush {
     bIsDynamicallyLoaded=false
     drawAs=Image
     imageType=FullColor
     mirroring=NoMirror
     resourceName=None
     resourceObject=null
     tiling=NoTile
     imageSize=<664 672>
     margin=<ltrb 0 0 0 0>
     outlineSettings {
      bUseBrushTransparency=false
      roundingType=HalfHeightRadius
      width=0
      color {colorUseRule=UseColor_Specified, specifiedColor=<rgba 0 0 0 0>}
      cornerRadii=<0 0 0 0>
     }
     tintColor {colorUseRule=UseColor_Specified, specifiedColor=<rgba 1 1 1 1>}
     uVRegion {bIsValid=false, max=<0 0>, min=<0 0>}
    }
    renderTransform {angle=0, scale=<0.7 0.7>, shear=<0 0>, translation=<0 0>}
   }
   slot {
    class=CanvasPanelSlot
    properties {layoutData {alignment=<0 0>, anchors {maximum=<0 0>, minimum=<0 0>}, offsets=<ltrb -171.2851 -108.1776 1067.1429 1080>}}
   }
  }
  RightPanel_Flames {
   widgetClassPath=&Image
   bIsVariable=true
   parent=RightPanelCanvas
   widget=RightPanel_Flames
   properties {
    brush {
     bIsDynamicallyLoaded=false
     drawAs=Image
     imageType=FullColor
     mirroring=NoMirror
     resourceName=None
     resourceObject=null
     tiling=NoTile
     imageSize=<400 368>
     margin=<ltrb 0 0 0 0>
     outlineSettings {
      bUseBrushTransparency=false
      roundingType=HalfHeightRadius
      width=0
      color {colorUseRule=UseColor_Specified, specifiedColor=<rgba 0 0 0 0>}
      cornerRadii=<0 0 0 0>
     }
     tintColor {colorUseRule=UseColor_Specified, specifiedColor=<rgba 1 1 1 1>}
     uVRegion {bIsValid=false, max=<0 0>, min=<0 0>}
    }
    renderTransform {angle=0, scale=<1.5 1.5>, shear=<0 0>, translation=<0 0>}
   }
   slot {
    class=CanvasPanelSlot
    properties {bAutoSize=true, layoutData {alignment=<0 0>, anchors {maximum=<0 0>, minimum=<0 0>}, offsets=<ltrb 358.9447 626.4836 400 368>}}
   }
  }
  RightPanel_House {
   widgetClassPath=&Image
   bIsVariable=true
   parent=RightPanelCanvas
   widget=RightPanel_House
   properties {
    brush {
     bIsDynamicallyLoaded=false
     drawAs=Image
     imageType=FullColor
     mirroring=NoMirror
     resourceName=None
     resourceObject=null
     tiling=NoTile
     imageSize=<580 444>
     margin=<ltrb 0 0 0 0>
     outlineSettings {
      bUseBrushTransparency=false
      roundingType=HalfHeightRadius
      width=0
      color {colorUseRule=UseColor_Specified, specifiedColor=<rgba 0 0 0 0>}
      cornerRadii=<0 0 0 0>
     }
     tintColor {colorUseRule=UseColor_Specified, specifiedColor=<rgba 1 1 1 1>}
     uVRegion {bIsValid=false, max=<0 0>, min=<0 0>}
    }
    renderTransform {angle=0, scale=<0.5 0.5>, shear=<0 0>, translation=<0 0>}
   }
   slot {
    class=CanvasPanelSlot
    properties {layoutData {alignment=<0 0>, anchors {maximum=<0 0>, minimum=<0 0>}, offsets=<ltrb -5.3006 270.7647 1067.1429 1080>}}
   }
  }
  RightPanel_Bang {
   widgetClassPath=&Image
   bIsVariable=true
   parent=RightPanelCanvas
   widget=RightPanel_Bang
   properties {brush {
     bIsDynamicallyLoaded=false
     drawAs=Image
     imageType=FullColor
     mirroring=NoMirror
     resourceName=None
     resourceObject=null
     tiling=NoTile
     imageSize=<244 196>
     margin=<ltrb 0 0 0 0>
     outlineSettings {
      bUseBrushTransparency=false
      roundingType=HalfHeightRadius
      width=0
      color {colorUseRule=UseColor_Specified, specifiedColor=<rgba 0 0 0 0>}
      cornerRadii=<0 0 0 0>
     }
     tintColor {colorUseRule=UseColor_Specified, specifiedColor=<rgba 1 1 1 1>}
     uVRegion {bIsValid=false, max=<0 0>, min=<0 0>}
    }}
   slot {
    class=CanvasPanelSlot
    properties {layoutData {alignment=<0 0>, anchors {maximum=<0 0>, minimum=<0 0>}, offsets=<ltrb 344.301 430.1842 244 196>}}
   }
  }
  RightPanel_Boom1 {
   widgetClassPath=&Image
   bIsVariable=true
   parent=RightPanelCanvas
   widget=RightPanel_Boom1
   properties {brush {
     bIsDynamicallyLoaded=false
     drawAs=Image
     imageType=FullColor
     mirroring=NoMirror
     resourceName=None
     resourceObject=null
     tiling=NoTile
     imageSize=<392 304>
     margin=<ltrb 0 0 0 0>
     outlineSettings {
      bUseBrushTransparency=false
      roundingType=HalfHeightRadius
      width=0
      color {colorUseRule=UseColor_Specified, specifiedColor=<rgba 0 0 0 0>}
      cornerRadii=<0 0 0 0>
     }
     tintColor {colorUseRule=UseColor_Specified, specifiedColor=<rgba 1 1 1 1>}
     uVRegion {bIsValid=false, max=<0 0>, min=<0 0>}
    }}
   slot {
    class=CanvasPanelSlot
    properties {layoutData {alignment=<0 0>, anchors {maximum=<0 0>, minimum=<0 0>}, offsets=<ltrb 238.1928 689.1035 392 304>}}
   }
  }
  RightPanel_Boom2 {
   widgetClassPath=&Image
   bIsVariable=true
   parent=RightPanelCanvas
   widget=RightPanel_Boom2
   properties {
    brush {
     bIsDynamicallyLoaded=false
     drawAs=Image
     imageType=FullColor
     mirroring=NoMirror
     resourceName=None
     resourceObject=null
     tiling=NoTile
     imageSize=<216 224>
     margin=<ltrb 0 0 0 0>
     outlineSettings {
      bUseBrushTransparency=false
      roundingType=HalfHeightRadius
      width=0
      color {colorUseRule=UseColor_Specified, specifiedColor=<rgba 0 0 0 0>}
      cornerRadii=<0 0 0 0>
     }
     tintColor {colorUseRule=UseColor_Specified, specifiedColor=<rgba 1 1 1 1>}
     uVRegion {bIsValid=false, max=<0 0>, min=<0 0>}
    }
    renderTransform {angle=0, scale=<0.8 0.8>, shear=<0 0>, translation=<0 0>}
   }
   slot {
    class=CanvasPanelSlot
    properties {layoutData {alignment=<0 0>, anchors {maximum=<0 0>, minimum=<0 0>}, offsets=<ltrb 171.1139 32.1003 216 224>}}
   }
  }
  RightPanel_Ping1 {
   widgetClassPath=&Image
   bIsVariable=true
   parent=RightPanelCanvas
   widget=RightPanel_Ping1
   properties {brush {
     bIsDynamicallyLoaded=false
     drawAs=Image
     imageType=FullColor
     mirroring=NoMirror
     resourceName=None
     resourceObject=null
     tiling=NoTile
     imageSize=<164 104>
     margin=<ltrb 0 0 0 0>
     outlineSettings {
      bUseBrushTransparency=false
      roundingType=HalfHeightRadius
      width=0
      color {colorUseRule=UseColor_Specified, specifiedColor=<rgba 0 0 0 0>}
      cornerRadii=<0 0 0 0>
     }
     tintColor {colorUseRule=UseColor_Specified, specifiedColor=<rgba 1 1 1 1>}
     uVRegion {bIsValid=false, max=<0 0>, min=<0 0>}
    }}
   slot {
    class=CanvasPanelSlot
    properties {layoutData {alignment=<0 0>, anchors {maximum=<0 0>, minimum=<0 0>}, offsets=<ltrb 602.8288 468.3297 164 104>}}
   }
  }
  RightPanel_Ping2 {
   widgetClassPath=&Image
   bIsVariable=true
   parent=RightPanelCanvas
   widget=RightPanel_Ping2
   properties {brush {
     bIsDynamicallyLoaded=false
     drawAs=Image
     imageType=FullColor
     mirroring=NoMirror
     resourceName=None
     resourceObject=null
     tiling=NoTile
     imageSize=<232 156>
     margin=<ltrb 0 0 0 0>
     outlineSettings {
      bUseBrushTransparency=false
      roundingType=HalfHeightRadius
      width=0
      color {colorUseRule=UseColor_Specified, specifiedColor=<rgba 0 0 0 0>}
      cornerRadii=<0 0 0 0>
     }
     tintColor {colorUseRule=UseColor_Specified, specifiedColor=<rgba 1 1 1 1>}
     uVRegion {bIsValid=false, max=<0 0>, min=<0 0>}
    }}
   slot {
    class=CanvasPanelSlot
    properties {layoutData {alignment=<0 0>, anchors {maximum=<0 0>, minimum=<0 0>}, offsets=<ltrb 466.8057 583.3495 232 156>}}
   }
  }
 ]
}
