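# PistolCanary comic hero panel 3 widget, animating layered left and right character panels that slide vertically through the screen while fading.
# PlayPistolCanaryHeroPanelAnimation selects the left or right animation from the HeroPanel enum, revealing images that Construct hides.
# LoadTextures assigns soft textures by array index, using low-res versions on mobile.
# UnloadTextures clears the brushes, then forces garbage collection and prints a debug message.
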WidgetBlueprint /PistolCanary/Layout/BP/ComicPanels/HeroPanels/PistolCanary_HeroPanel3
animations [
 LeftPanelAnim {
  binding_count=3
  length_seconds=9.45
  object=&self:LeftPanelAnim
  sequence {
   bindings [
    LeftPanel_Chr {
     id=#1
     possessed_class=Image
     tracks [
      {
       type=MovieScene2DTransformTrack
       sections [{
         type=MovieScene2DTransformSection
         end_frame=566
         start_frame=0
         channels [
          Translation.X {type=MovieSceneScriptingFloatChannel, default=0}
          Translation.Y {type=MovieSceneScriptingFloatChannel, keys [[0,129.0447],[566,0]]}
          Angle {type=MovieSceneScriptingFloatChannel, default=0}
          Scale.X {type=MovieSceneScriptingFloatChannel, default=1, keys [[0,0.9],[566,1]]}
          Scale.Y {type=MovieSceneScriptingFloatChannel, default=1, keys [[0,0.9],[566,1]]}
          Shear.X {type=MovieSceneScriptingFloatChannel, default=0}
          Shear.Y {type=MovieSceneScriptingFloatChannel, default=0}
         ]
        }]
      }
      {
       type=MovieSceneFloatTrack
       display_name=Render Opacity
       sections [{type=MovieSceneFloatSection, end_frame=19, start_frame=0, channels [None {type=MovieSceneScriptingFloatChannel, const [1,1]}]}]
      }
     ]
    }
    LeftPanelCanvas {
     id=#2
     possessed_class=CanvasPanel
     tracks [
      {
       type=MovieSceneFloatTrack
       display_name=Render Opacity
       sections [{
         type=MovieSceneFloatSection
         end_frame=567
         start_frame=0
         channels [None {type=MovieSceneScriptingFloatChannel, keys [[0,0],[16,1],[552,1],[566,0]]}]
        }]
      }
      {
       type=MovieScene2DTransformTrack
       sections [{
         type=MovieScene2DTransformSection
         end_frame=567
         start_frame=0
         channels [
          Translation.X {type=MovieSceneScriptingFloatChannel, thinned_from=4, keys [[0,-308.7808],[16,0],[566,0]]}
          Translation.Y {type=MovieSceneScriptingFloatChannel, keys [[0,1088.9025],[16,0],[552,0],[566,-1091.6603]]}
          Angle {type=MovieSceneScriptingFloatChannel, const [0,2]}
          Scale.X {type=MovieSceneScriptingFloatChannel, const [1,2]}
          Scale.Y {type=MovieSceneScriptingFloatChannel, const [1,2]}
          Shear.X {type=MovieSceneScriptingFloatChannel, const [0,2]}
          Shear.Y {type=MovieSceneScriptingFloatChannel, const [0,2]}
         ]
        }]
      }
     ]
    }
    LeftPanel_BG {
     id=#3
     possessed_class=Image
     tracks [{
       type=MovieSceneFloatTrack
       display_name=Render Opacity
       sections [{type=MovieSceneFloatSection, end_frame=19, start_frame=0, channels [None {type=MovieSceneScriptingFloatChannel, const [1,1]}]}]
      }]
    }
   ]
   display_rate {numerator=60}
  }
 }
 RightPanelAnim {
  binding_count=7
  length_seconds=8.0667
  object=&self:RightPanelAnim
  sequence {
   bindings [
    RightPanel_BirdBottom {
     id=#4
     possessed_class=Image
     tracks [
      {
       type=MovieScene2DTransformTrack
       sections [{
         type=MovieScene2DTransformSection
         end_frame=476
         start_frame=5
         channels [
          Translation.X {type=MovieSceneScriptingFloatChannel, default=0}
          Translation.Y {type=MovieSceneScriptingFloatChannel, keys [[5,0],[476,20.9535]]}
          Angle {type=MovieSceneScriptingFloatChannel, default=0}
          Scale.X {type=MovieSceneScriptingFloatChannel, default=0.7}
          Scale.Y {type=MovieSceneScriptingFloatChannel, default=0.7}
          Shear.X {type=MovieSceneScriptingFloatChannel, default=0}
          Shear.Y {type=MovieSceneScriptingFloatChannel, default=0}
         ]
        }]
      }
      {
       type=MovieSceneFloatTrack
       display_name=Render Opacity
       sections [{type=MovieSceneFloatSection, end_frame=19, start_frame=0, channels [None {type=MovieSceneScriptingFloatChannel, const [1,1]}]}]
      }
     ]
    }
    RightPanel_Smoke {
     id=#5
     possessed_class=Image
     tracks [
      {
       type=MovieScene2DTransformTrack
       sections [{
         type=MovieScene2DTransformSection
         end_frame=476
         start_frame=5
         channels [
          Translation.X {type=MovieSceneScriptingFloatChannel, default=7.5969}
          Translation.Y {type=MovieSceneScriptingFloatChannel, default=0, keys [[5,62.8109],[476,0]]}
          Angle {type=MovieSceneScriptingFloatChannel, default=0}
          Scale.X {type=MovieSceneScriptingFloatChannel, keys [[5,1.2],[476,1.3]]}
          Scale.Y {type=MovieSceneScriptingFloatChannel, keys [[5,1.2],[476,1.3]]}
          Shear.X {type=MovieSceneScriptingFloatChannel, default=0}
          Shear.Y {type=MovieSceneScriptingFloatChannel, default=0}
         ]
        }]
      }
      {
       type=MovieSceneFloatTrack
       display_name=Render Opacity
       sections [{type=MovieSceneFloatSection, end_frame=19, start_frame=0, channels [None {type=MovieSceneScriptingFloatChannel, const [1,1]}]}]
      }
     ]
    }
    RightPanelCanvas {
     id=#6
     possessed_class=CanvasPanel
     tracks [
      {
       type=MovieSceneFloatTrack
       display_name=Render Opacity
       sections [{
         type=MovieSceneFloatSection
         end_frame=484
         start_frame=1
         channels [None {type=MovieSceneScriptingFloatChannel, keys [[1,0],[16,1],[467,1],[483,0]]}]
        }]
      }
      {
       type=MovieScene2DTransformTrack
       sections [{
         type=MovieScene2DTransformSection
         end_frame=484
         start_frame=1
         channels [
          Translation.X {type=MovieSceneScriptingFloatChannel, thinned_from=4, keys [[1,353.58],[16,0],[483,0]]}
          Translation.Y {type=MovieSceneScriptingFloatChannel, keys [[1,-1081.0922],[16,0],[467,0],[483,1082.3345]]}
          Angle {type=MovieSceneScriptingFloatChannel, const [0,2]}
          Scale.X {type=MovieSceneScriptingFloatChannel, const [1,2]}
          Scale.Y {type=MovieSceneScriptingFloatChannel, const [1,2]}
          Shear.X {type=MovieSceneScriptingFloatChannel, const [0,2]}
          Shear.Y {type=MovieSceneScriptingFloatChannel, const [0,2]}
         ]
        }]
      }
     ]
    }
    RightPanel_Chr {
     id=#7
     possessed_class=Image
     tracks [
      {
       type=MovieScene2DTransformTrack
       sections [{
         type=MovieScene2DTransformSection
         end_frame=476
         start_frame=5
         channels [
          Translation.X {type=MovieSceneScriptingFloatChannel, default=0}
          Translation.Y {type=MovieSceneScriptingFloatChannel, keys [[5,0],[476,-19.7799]]}
          Angle {type=MovieSceneScriptingFloatChannel, default=0}
          Scale.X {type=MovieSceneScriptingFloatChannel, keys [[5,0.97],[476,1]]}
          Scale.Y {type=MovieSceneScriptingFloatChannel, keys [[5,0.97],[476,1]]}
          Shear.X {type=MovieSceneScriptingFloatChannel, default=0}
          Shear.Y {type=MovieSceneScriptingFloatChannel, default=0}
         ]
        }]
      }
      {
       type=MovieSceneFloatTrack
       display_name=Render Opacity
       sections [{type=MovieSceneFloatSection, end_frame=19, start_frame=0, channels [None {type=MovieSceneScriptingFloatChannel, const [1,1]}]}]
      }
     ]
    }
    RightPanel_BirdTop {
     id=#8
     possessed_class=Image
     tracks [
      {
       type=MovieScene2DTransformTrack
       sections [{
         type=MovieScene2DTransformSection
         end_frame=476
         start_frame=5
         channels [
          Translation.X {type=MovieSceneScriptingFloatChannel, default=0}
          Translation.Y {type=MovieSceneScriptingFloatChannel, default=0}
          Angle {type=MovieSceneScriptingFloatChannel, default=0}
          Scale.X {type=MovieSceneScriptingFloatChannel, keys [[5,0.7],[476,0.75]]}
          Scale.Y {type=MovieSceneScriptingFloatChannel, keys [[5,0.7],[476,0.75]]}
          Shear.X {type=MovieSceneScriptingFloatChannel, default=0}
          Shear.Y {type=MovieSceneScriptingFloatChannel, default=0}
         ]
        }]
      }
      {
       type=MovieSceneFloatTrack
       display_name=Render Opacity
       sections [{type=MovieSceneFloatSection, end_frame=19, start_frame=0, channels [None {type=MovieSceneScriptingFloatChannel, const [1,1]}]}]
      }
     ]
    }
    RightPanel_BirdGroup {
     id=#9
     possessed_class=Image
     tracks [
      {
       type=MovieScene2DTransformTrack
       sections [{
         type=MovieScene2DTransformSection
         end_frame=484
         start_frame=5
         channels [
          Translation.X {type=MovieSceneScriptingFloatChannel, default=0}
          Translation.Y {type=MovieSceneScriptingFloatChannel, default=0}
          Angle {type=MovieSceneScriptingFloatChannel, keys [[5,0],[476,2.9852]]}
          Scale.X {type=MovieSceneScriptingFloatChannel, default=1, const [1.03,2]}
          Scale.Y {type=MovieSceneScriptingFloatChannel, default=1, const [1.03,2]}
          Shear.X {type=MovieSceneScriptingFloatChannel, default=0}
          Shear.Y {type=MovieSceneScriptingFloatChannel, default=0}
         ]
        }]
      }
      {
       type=MovieSceneFloatTrack
       display_name=Render Opacity
       sections [{type=MovieSceneFloatSection, end_frame=19, start_frame=0, channels [None {type=MovieSceneScriptingFloatChannel, const [1,1]}]}]
      }
     ]
    }
    RightPanel_BG {
     id=#10
     possessed_class=Image
     tracks [{
       type=MovieSceneFloatTrack
       display_name=Render Opacity
       sections [{type=MovieSceneFloatSection, end_frame=19, start_frame=0, channels [None {type=MovieSceneScriptingFloatChannel, const [1,1]}]}]
      }]
    }
   ]
   display_rate {numerator=60}
  }
 }
]
graphs [EventGraph {nodes [
   Event Construct()
   > SetRenderOpacity(
    self=[
     Get RightPanel_Smoke() -> (RightPanel_Smoke:Image Object Reference)
     Get RightPanel_Chr() -> (RightPanel_Chr:Image Object Reference)
     Get RightPanel_BirdTop() -> (RightPanel_BirdTop:Image Object Reference)
     Get RightPanel_BG() -> (RightPanel_BG:Image Object Reference)
     Get LeftPanel_BG() -> (LeftPanel_BG:Image Object Reference)
     Get RightPanel_BirdGroup() -> (RightPanel_BirdGroup:Image Object Reference)
     Get RightPanel_BirdBottom() -> (RightPanel_BirdBottom:Image Object Reference)
     Get LeftPanel_Chr() -> (LeftPanel_Chr:Image Object Reference)
    ]
    InOpacity=0
   )
   Event HeroPanel_UnloadTextures() {is=title K2Node_Event}
   n1 = Make Array!(
    "[0]":Image Object Reference=Get LeftPanel_BG()
    "[1]":Image Object Reference=Get LeftPanel_Chr()
    "[2]":Image Object Reference=Get RightPanel_BG()
    "[3]":Image Object Reference=Get RightPanel_Chr()
    "[4]":Image Object Reference=Get RightPanel_Smoke()
    "[5]":Image Object Reference=Get RightPanel_BirdTop()
    "[6]":Image Object Reference=Get RightPanel_BirdGroup()
    "[7]":Image Object Reference=Get RightPanel_BirdBottom()
   ) -> (Array:Array of Image Object References)
   > n2 = For Each Loop(Array:Array of Image Object References=@n1) -> (Array Element:Image Object Reference)
   CollectGarbage!(execute=@n2.Completed) -> (then)
   > PrintString!(execute, InString=UNLOADED HERO PANEL 3 TEXTURES, bPrintToScreen, bPrintToLog, TextColor=(R=0.9931,G=0.9931,B=1,A=1), Duration=3, Key) -> (then)
   Is Valid(exec=@n2.LoopBody, InputObject=@n2)
   > SetBrushFromTexture!(execute, self:Image Object Reference=@n2, Texture:Texture 2D Object Reference, bMatchSize:bool=true) -> (then)
   n3 = Event PlayPistolCanaryHeroPanelAnimation!() -> (OutputDelegate, then, HeroPanel:E_PistolCanary_HeroPanelSelect Enum=NewEnumerator0, PanelSide:E_PistolCanary_PanelSide Enum=NewEnumerator0) {is=title K2Node_Event}
   > PlayAnimation(InAnimation=Select!(
     NewEnumerator0:Widget Animation Object Reference
     NewEnumerator1:Widget Animation Object Reference
     NewEnumerator2:Widget Animation Object Reference
     NewEnumerator3:Widget Animation Object Reference
     NewEnumerator4:Widget Animation Object Reference=Get LeftPanelAnim() -> (LeftPanelAnim:Widget Animation Object Reference)
     NewEnumerator5:Widget Animation Object Reference=Get RightPanelAnim() -> (RightPanelAnim:Widget Animation Object Reference)
     Index:E_PistolCanary_HeroPanelSelect Enum=@n3.HeroPanel
    ) -> (ReturnValue:Widget Animation Object Reference))
   Event HeroPanel_LoadTextures() {is=title K2Node_Event}
   > n4 = For Each Loop(Array=@n1)
   n5 = IsMobilePlatform!() -> (ReturnValue:bool)
   n6 = Get TexturesLow() -> (TexturesLow:Array of Texture 2D Soft Object References)
   n7 = Get Textures() -> (Textures:Array of Texture 2D Soft Object References)
   > SetBrushFromSoftTexture!(
    execute
    self:Image Object Reference=@n4
    SoftTexture:Texture 2D Soft Object Reference="Get (a copy)"(
     Array:Array of Texture 2D Soft Object References=Select!(Option 0:Array of Texture 2D Soft Object References=@n7, Option 1:Array of Texture 2D Soft Object References=@n6, Index:bool=@n5) -> (ReturnValue:Array of Texture 2D Soft Object References)
     Dimension 1=@n4.Array Index
    ) -> (Output:Texture 2D Soft Object Reference)
    bMatchSize:bool=false
   ) -> (then)
   Event PreConstruct()
   Event Tick()
  ]}]
variables [
 Textures {
  type="array<Texture2D>"
  default [
   /PistolCanary/Layout/Texture/ComicPanels/HeroPanels/Hero3/FN_Comic_Overlay_D1_background
   /PistolCanary/Layout/Texture/ComicPanels/HeroPanels/Hero3/FN_Comic_Overlay_D1_character
   /PistolCanary/Layout/Texture/ComicPanels/HeroPanels/Hero3/FN_Comic_Overlay_D2_background_01
   /PistolCanary/Layout/Texture/ComicPanels/HeroPanels/Hero3/FN_Comic_Overlay_D2_character
   /PistolCanary/Layout/Texture/ComicPanels/HeroPanels/Hero3/FN_Comic_Overlay_D2_background_02
   /PistolCanary/Layout/Texture/ComicPanels/HeroPanels/Hero3/FN_Comic_Overlay_D2_foreground_crow_top
   /PistolCanary/Layout/Texture/ComicPanels/HeroPanels/Hero3/FN_Comic_Overlay_D2_background_crows
   /PistolCanary/Layout/Texture/ComicPanels/HeroPanels/Hero3/FN_Comic_Overlay_D2_foreground_crow_bottom
  ]
 }
 TexturesLow {
  type="array<Texture2D>"
  default [
   /PistolCanary/Layout/Texture/ComicPanels/HeroPanels/Hero3/LowRes/FN_Comic_Overlay_D1_background_low
   /PistolCanary/Layout/Texture/ComicPanels/HeroPanels/Hero3/LowRes/FN_Comic_Overlay_D1_character_low
   /PistolCanary/Layout/Texture/ComicPanels/HeroPanels/Hero3/LowRes/FN_Comic_Overlay_D2_background_01_low
   /PistolCanary/Layout/Texture/ComicPanels/HeroPanels/Hero3/LowRes/FN_Comic_Overlay_D2_character_low
   /PistolCanary/Layout/Texture/ComicPanels/HeroPanels/Hero3/LowRes/FN_Comic_Overlay_D2_background_02_low
   /PistolCanary/Layout/Texture/ComicPanels/HeroPanels/Hero3/LowRes/FN_Comic_Overlay_D2_foreground_crow_top_low
   /PistolCanary/Layout/Texture/ComicPanels/HeroPanels/Hero3/LowRes/FN_Comic_Overlay_D2_background_crows_low
   /PistolCanary/Layout/Texture/ComicPanels/HeroPanels/Hero3/LowRes/FN_Comic_Overlay_D2_foreground_crow_bottom_low
  ]
 }
]
widgets {
 info {inheritedWidgetCount=0, parentClass=&UserWidget, rootWidgetClass=&CanvasPanel, widgetCount=11}
 widgets [
  CanvasPanel_163 {widgetClassPath=&CanvasPanel, widget=CanvasPanel_163, properties {clipping=ClipToBounds, visibility=HitTestInvisible}}
  LeftPanelCanvas {
   widgetClassPath=&CanvasPanel
   parent=CanvasPanel_163
   widget=LeftPanelCanvas
   properties {visibility=HitTestInvisible}
   slot {
    class=CanvasPanelSlot
    properties {bAutoSize=true, layoutData {alignment=<0 0>, anchors {maximum=<0 0.5>, minimum=<0 0.5>}, offsets=<ltrb 0 -540.5405 672 1024>}}
   }
  }
  LeftPanel_BG {
   widgetClassPath=&Image
   bIsVariable=true
   parent=LeftPanelCanvas
   widget=LeftPanel_BG
   properties {brush {
     bIsDynamicallyLoaded=false
     drawAs=Image
     imageType=FullColor
     mirroring=NoMirror
     resourceName=None
     resourceObject=null
     tiling=NoTile
     imageSize=<672 1024>
     margin=<ltrb 0 0 0 0>
     outlineSettings {
      bUseBrushTransparency=false
      roundingType=HalfHeightRadius
      width=0
      color {colorUseRule=UseColor_Specified, specifiedColor=<rgba 0 0 0 0>}
      cornerRadii=<0 0 0 0>
     }
     tintColor {colorUseRule=UseColor_Specified, specifiedColor=<rgba 1 1 1 1>}
     uVRegion {bIsValid=false, max=<0 0>, min=<0 0>}
    }}
   slot {class=CanvasPanelSlot, properties {layoutData {alignment=<0 0>, anchors {maximum=<0 0>, minimum=<0 0>}, offsets=<ltrb 0 0 708.75 1080>}}}
  }
  LeftPanel_Chr {
   widgetClassPath=&Image
   bIsVariable=true
   parent=LeftPanelCanvas
   widget=LeftPanel_Chr
   properties {brush {
     bIsDynamicallyLoaded=false
     drawAs=Image
     imageType=FullColor
     mirroring=NoMirror
     resourceName=None
     resourceObject=null
     tiling=NoTile
     imageSize=<840 1132>
     margin=<ltrb 0 0 0 0>
     outlineSettings {
      bUseBrushTransparency=false
      roundingType=HalfHeightRadius
      width=0
      color {colorUseRule=UseColor_Specified, specifiedColor=<rgba 0 0 0 0>}
      cornerRadii=<0 0 0 0>
     }
     tintColor {colorUseRule=UseColor_Specified, specifiedColor=<rgba 1 1 1 1>}
     uVRegion {bIsValid=false, max=<0 0>, min=<0 0>}
    }}
   slot {
    class=CanvasPanelSlot
    properties {layoutData {alignment=<0 0>, anchors {maximum=<0 0>, minimum=<0 0>}, offsets=<ltrb -79.3991 0 801.4134 1080>}}
   }
  }
  RightPanelCanvas {
   widgetClassPath=&CanvasPanel
   parent=CanvasPanel_163
   widget=RightPanelCanvas
   properties {visibility=HitTestInvisible}
   slot {
    class=CanvasPanelSlot
    properties {bAutoSize=true, layoutData {alignment=<0 0>, anchors {maximum=<1 0.5>, minimum=<1 0.5>}, offsets=<ltrb -751 -540.5405 100 30>}}
   }
  }
  RightPanel_BG {
   widgetClassPath=&Image
   bIsVariable=true
   parent=RightPanelCanvas
   widget=RightPanel_BG
   properties {brush {
     bIsDynamicallyLoaded=false
     drawAs=Image
     imageType=FullColor
     mirroring=NoMirror
     resourceName=None
     resourceObject=null
     tiling=NoTile
     imageSize=<712 1024>
     margin=<ltrb 0 0 0 0>
     outlineSettings {
      bUseBrushTransparency=false
      roundingType=HalfHeightRadius
      width=0
      color {colorUseRule=UseColor_Specified, specifiedColor=<rgba 0 0 0 0>}
      cornerRadii=<0 0 0 0>
     }
     tintColor {colorUseRule=UseColor_Specified, specifiedColor=<rgba 1 1 1 1>}
     uVRegion {bIsValid=false, max=<0 0>, min=<0 0>}
    }}
   slot {class=CanvasPanelSlot, properties {layoutData {alignment=<0 0>, anchors {maximum=<0 0>, minimum=<0 0>}, offsets=<ltrb 0 0 751 1080>}}}
  }
  RightPanel_Smoke {
   widgetClassPath=&Image
   bIsVariable=true
   parent=RightPanelCanvas
   widget=RightPanel_Smoke
   properties {
    brush {
     bIsDynamicallyLoaded=false
     drawAs=Image
     imageType=FullColor
     mirroring=NoMirror
     resourceName=None
     resourceObject=null
     tiling=NoTile
     imageSize=<356 400>
     margin=<ltrb 0 0 0 0>
     outlineSettings {
      bUseBrushTransparency=false
      roundingType=HalfHeightRadius
      width=0
      color {colorUseRule=UseColor_Specified, specifiedColor=<rgba 0 0 0 0>}
      cornerRadii=<0 0 0 0>
     }
     tintColor {colorUseRule=UseColor_Specified, specifiedColor=<rgba 1 1 1 1>}
     uVRegion {bIsValid=false, max=<0 0>, min=<0 0>}
    }
    renderTransform {angle=0, scale=<1.3 1.3>, shear=<0 0>, translation=<0 0>}
   }
   slot {
    class=CanvasPanelSlot
    properties {zOrder=1, layoutData {alignment=<0 0>, anchors {maximum=<0 0>, minimum=<0 0>}, offsets=<ltrb 354.2857 633.2747 356 400>}}
   }
  }
  RightPanel_BirdGroup {
   widgetClassPath=&Image
   bIsVariable=true
   parent=RightPanelCanvas
   widget=RightPanel_BirdGroup
   properties {brush {
     bIsDynamicallyLoaded=false
     drawAs=Image
     imageType=FullColor
     mirroring=NoMirror
     resourceName=None
     resourceObject=null
     tiling=NoTile
     imageSize=<808 724>
     margin=<ltrb 0 0 0 0>
     outlineSettings {
      bUseBrushTransparency=false
      roundingType=HalfHeightRadius
      width=0
      color {colorUseRule=UseColor_Specified, specifiedColor=<rgba 0 0 0 0>}
      cornerRadii=<0 0 0 0>
     }
     tintColor {colorUseRule=UseColor_Specified, specifiedColor=<rgba 1 1 1 1>}
     uVRegion {bIsValid=false, max=<0 0>, min=<0 0>}
    }}
   slot {
    class=CanvasPanelSlot
    properties {zOrder=1, layoutData {alignment=<0 0>, anchors {maximum=<0 0>, minimum=<0 0>}, offsets=<ltrb -80.0504 71.063 800 724>}}
   }
  }
  RightPanel_BirdTop {
   widgetClassPath=&Image
   bIsVariable=true
   parent=RightPanelCanvas
   widget=RightPanel_BirdTop
   properties {
    brush {
     bIsDynamicallyLoaded=false
     drawAs=Image
     imageType=FullColor
     mirroring=NoMirror
     resourceName=None
     resourceObject=null
     tiling=NoTile
     imageSize=<772 480>
     margin=<ltrb 0 0 0 0>
     outlineSettings {
      bUseBrushTransparency=false
      roundingType=HalfHeightRadius
      width=0
      color {colorUseRule=UseColor_Specified, specifiedColor=<rgba 0 0 0 0>}
      cornerRadii=<0 0 0 0>
     }
     tintColor {colorUseRule=UseColor_Specified, specifiedColor=<rgba 1 1 1 1>}
     uVRegion {bIsValid=false, max=<0 0>, min=<0 0>}
    }
    renderTransform {angle=0, scale=<0.7 0.7>, shear=<0 0>, translation=<0 0>}
   }
   slot {
    class=CanvasPanelSlot
    properties {zOrder=2, layoutData {alignment=<0 0>, anchors {maximum=<0 0>, minimum=<0 0>}, offsets=<ltrb 39.8361 -153.7417 888.9594 480>}}
   }
  }
  RightPanel_BirdBottom {
   widgetClassPath=&Image
   bIsVariable=true
   parent=RightPanelCanvas
   widget=RightPanel_BirdBottom
   properties {
    brush {
     bIsDynamicallyLoaded=false
     drawAs=Image
     imageType=FullColor
     mirroring=NoMirror
     resourceName=None
     resourceObject=null
     tiling=NoTile
     imageSize=<564 480>
     margin=<ltrb 0 0 0 0>
     outlineSettings {
      bUseBrushTransparency=false
      roundingType=HalfHeightRadius
      width=0
      color {colorUseRule=UseColor_Specified, specifiedColor=<rgba 0 0 0 0>}
      cornerRadii=<0 0 0 0>
     }
     tintColor {colorUseRule=UseColor_Specified, specifiedColor=<rgba 1 1 1 1>}
     uVRegion {bIsValid=false, max=<0 0>, min=<0 0>}
    }
    renderTransform {angle=0, scale=<0.7 0.7>, shear=<0 0>, translation=<0 0>}
   }
   slot {
    class=CanvasPanelSlot
    properties {zOrder=2, layoutData {alignment=<0 0>, anchors {maximum=<0 0>, minimum=<0 0>}, offsets=<ltrb -68.0717 588.889 564 480>}}
   }
  }
  RightPanel_Chr {
   widgetClassPath=&Image
   bIsVariable=true
   parent=RightPanelCanvas
   widget=RightPanel_Chr
   properties {brush {
     bIsDynamicallyLoaded=false
     drawAs=Image
     imageType=FullColor
     mirroring=NoMirror
     resourceName=None
     resourceObject=null
     tiling=NoTile
     imageSize=<400 1044>
     margin=<ltrb 0 0 0 0>
     outlineSettings {
      bUseBrushTransparency=false
      roundingType=HalfHeightRadius
      width=0
      color {colorUseRule=UseColor_Specified, specifiedColor=<rgba 0 0 0 0>}
      cornerRadii=<0 0 0 0>
     }
     tintColor {colorUseRule=UseColor_Specified, specifiedColor=<rgba 1 1 1 1>}
     uVRegion {bIsValid=false, max=<0 0>, min=<0 0>}
    }}
   slot {
    class=CanvasPanelSlot
    properties {zOrder=2, layoutData {alignment=<0 0>, anchors {maximum=<0 0>, minimum=<0 0>}, offsets=<ltrb 201.3817 101.0541 400 1044>}}
   }
  }
 ]
}
